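# Sound cue for the flying Battle Bus drop sound effect in Athena.
# A quality level node at the root chooses one of four wave players, the first two using the standard drop wave and the last two its _opt variant.
# The cue routes to the Battle Bus sound class and borrows the shared weapon-drop attenuation asset for spatial falloff.
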SoundCue /Game/Athena/Sounds/Bus/FlyingBus_Drop_Cue
root=QualityLevel_0
nodes [
 {id=QualityLevel_0, children [WavePlayer_0,WavePlayer_2,WavePlayer_1,WavePlayer_3]}
 {id=WavePlayer_0, properties {soundWaveAssetPtr=&/SFX_FlyingBus_Drop_03_16}}
 {id=WavePlayer_1, properties {soundWaveAssetPtr=&/SFX_FlyingBus_Drop_03_16_opt}}
 {id=WavePlayer_2, properties {soundWaveAssetPtr=&SFX_FlyingBus_Drop_03_16}}
 {id=WavePlayer_3, properties {soundWaveAssetPtr=&SFX_FlyingBus_Drop_03_16_opt}}
]
properties {
 attenuationSettings=&Fort_Weapon_Drop_Att
 maxDistance=3000
 soundClassObject=/Game/Sounds/FortSoundClassesAndModes/Fort_BattleBus
 volumeMultiplier=1
}
