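# Sound cue for the jump layer heard when a teammate drops from the flying Battle Bus, played through the Battle Bus sound class.
# A quality-level node sends its first two slots to a random pick of three wave variants and its last two to a pick of two.
# The two-wave random node is excluded from branch culling.
# Overridden attenuation sets long-range distance falloff and low-pass filter limits.
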SoundCue /Game/Athena/Sounds/Bus/FlyingBus_Drop_Teammate_Cue
root=QualityLevel_1
nodes [
 {id=QualityLevel_1, children [Random_0,Random_0,Random_1,Random_1]}
 {id=Random_0, children [WavePlayer_4,WavePlayer_5,WavePlayer_6], properties {weights [1,1,1]}}
 {id=Random_1, children [WavePlayer_0,WavePlayer_1], properties {bShouldExcludeFromBranchCulling=true, weights [1,1]}}
 {id=WavePlayer_0, properties {soundWaveAssetPtr=&/athena_busdrop_teammate_jumplayer_01}}
 {id=WavePlayer_1, properties {soundWaveAssetPtr=&/athena_busdrop_teammate_jumplayer_02}}
 {id=WavePlayer_4, properties {soundWaveAssetPtr=&/athena_busdrop_teammate_jumplayer_03}}
 {id=WavePlayer_5, properties {soundWaveAssetPtr=&athena_busdrop_teammate_jumplayer_01}}
 {id=WavePlayer_6, properties {soundWaveAssetPtr=&athena_busdrop_teammate_jumplayer_02}}
]
properties {
 bOverrideAttenuation=true
 maxDistance=11000
 soundClassObject=/Game/Sounds/FortSoundClassesAndModes/Fort_BattleBus
 volumeMultiplier=0.875
 attenuationOverrides {falloffDistance=10000, lPFFrequencyAtMax=3000, lPFRadiusMax=30000, attenuationShapeExtents {x=1000}}
}
